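# Sound cue for a coin appearing in the Wax limited-time mode.
# A Random node picks among three equally weighted appear waves and feeds a root Modulator left at default settings.
# The cue plays at slightly reduced volume through a 2D hit-notification sound class, with an attenuation override that sets a long falloff and stereo spread.
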SoundCue /Game/Athena/Sounds/LTMs/Wax/Wax_Coin_Appear_Cue
root=Modulator_0
nodes [
 {id=Modulator_0, children [Random_0]}
 {id=Random_0, children [WavePlayer_0,WavePlayer_1,WavePlayer_2], properties {weights [1,1,1]}}
 {id=WavePlayer_0, properties {soundWaveAssetPtr=&/Wax_Coin_Appear_01}}
 {id=WavePlayer_1, properties {soundWaveAssetPtr=&/Wax_Coin_Appear_02}}
 {id=WavePlayer_2, properties {soundWaveAssetPtr=&/Wax_Coin_Appear_03}}
]
properties {
 bOverrideAttenuation=true
 maxDistance=10000
 soundClassObject=&Fort_2DHitNotifications_Instigate_PSM
 volumeMultiplier=0.8
 attenuationOverrides {falloffDistance=10000, stereoSpread=512, attenuationShapeExtents {x=0}}
}
